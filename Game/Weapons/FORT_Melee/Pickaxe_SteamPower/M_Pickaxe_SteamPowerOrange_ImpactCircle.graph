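# Unlit translucent Niagara material for the orange SteamPower pickaxe's impact circle, drawing a noisy, toon-fire radial effect.
# Polar UVs, distorted by panning noise via dynamic parameters, form a radial mask eroded by scrolling wispy noise, sampled with explicit derivatives.
# The result indexes a fire curve-atlas gradient tinted by particle color for emissive.
# It also drives smoothstepped opacity with particle alpha, skipping depth fade on forward and mobile.
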Material /Game/Weapons/FORT_Melee/Pickaxe_SteamPower/M_Pickaxe_SteamPowerOrange_ImpactCircle
graph {
 properties {
  bUsedWithNiagaraMeshParticles=true
  bUsedWithNiagaraRibbons=true
  bUsedWithNiagaraSprites=true
  blendMode=BLEND_Translucent
  pixelDepthOffsetMode=PDOM_Legacy
  shadingModel=MSM_Unlit
  twoSided=true
  usedShadingModels=Unlit
 }
 nodes [
  n1 = ParticleColor()
  n2 = MaterialFunctionCall!(UV, Twist Center, Twist Edges) {materialFunction=/Game/Accessories/FORT_Backpacks/Backpack_F_MED_Celestial_Glow/FX/Materials/MF_PolarUV}
  n3 = TextureCoordinate()
  n4 = TextureSample(UVs=Panner(Coordinate=Multiply(A=@n3) {constB=0.5}) {speedY=0.5}) {samplerType=SAMPLERTYPE_Grayscale, texture=/Game/UI/Frontend/ItemShop/Art/Rumble/LowResBlurredNoise_Rumble}
  n5 = Add(
   A=@n2.Polar UVs
   B=Multiply(
    A=Subtract(A=@n4.R) {constB=0.5}
    B=DynamicParameter() {defaultValue=<rgba 0.1 2 0.1 0.1>, paramNames [Radius,Power,Distortion,Thickness]}.Distortion
   ) {constB=0}
  )
  n6 = ComponentMask(None=OneMinus(None=ComponentMask(None=@n5) {g=true})) {r=true}
  n7 = AppendVector(
   A=ScalarParameter() {defaultValue=1, parameterName=Scale X, sliderMax=5}
   B=ScalarParameter() {defaultValue=1, parameterName=Scale Y, sliderMax=5}
  )
  n8 = ScalarParameter() {parameterName=Speed X, sliderMax=5}
  n9 = ScalarParameter() {defaultValue=-0.5, parameterName=Speed Y, sliderMax=5}
  n10 = TextureSample!(
   UVs=Panner(Coordinate=Multiply(A=@n5, B=@n7) {constB=2}, Speed=AppendVector(A=@n8, B=@n9)) {speedY=-0.5}
   Tex
   "DDX(UVs)"=@n2.DDX
   "DDY(UVs)"=@n2.DDY
   Apply View MipBias
  ) {mipValueMode=TMVM_Derivative, texture=/Game/Effects/Fort_Effects/Textures/T_WispyFireNoise03}
  n11 = LinearInterpolate(
   A=Subtract(A=@n6, B=@n10.R)
   B=@n6
   Alpha=Power(Base=@n6, Exp=DynamicParameter() {defaultValue=<rgba 0.1 2 0.1 0.1>, paramNames [Radius,Power,Distortion,Thickness]}.Power)
  )
  n12 = SmoothStep(
   Min=ScalarParameter() {defaultValue=0.3, parameterName=Color SmoothStep Min, sliderMax=3, sliderMin=-3}
   Max=ScalarParameter() {defaultValue=1.2, parameterName=Color SmoothStep Max, sliderMax=3, sliderMin=-3}
   Value=Power(Base=@n11, Exp=ScalarParameter() {defaultValue=1, parameterName=Color Power, sliderMax=5})
  ) {constMax=1.2, constMin=0.3}
  MaterialFunctionCall(Emissive=Multiply(
    A=@n1.RGB
    B=CurveAtlasRowParameter(CurveTime=@n12) {atlas=/Game/Athena/LinearColorCurves/VFX_CurveAtlas_Toon, curve=/Game/Athena/LinearColorCurves/Curve_FireToon02}
   )) {materialFunction=&MF_FX_EmissiveAdjustments}
  Multiply(A=@n6, B=@n10.R) {constB=2}
  n13 = Multiply(A=@n1.A, B=@n11)
  n14 = DepthFade(Opacity=@n13, FadeDistance=ScalarParameter() {defaultValue=35, parameterName=Depth Fade Distance, sliderMax=100})
  SmoothStep(Max=ScalarParameter() {defaultValue=0.8, parameterName=Max}, Value=ShadingPathSwitch(Default=@n14, Deferred=@n14, Forward=@n13, Mobile=@n13))
 ]
}
